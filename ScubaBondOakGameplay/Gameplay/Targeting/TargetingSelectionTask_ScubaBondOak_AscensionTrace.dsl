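; Targeting selection trace task, apparently for an Ascension move, that aims a trace diagonally up and forward from the source actor.
; A helper builds the travel offset from world up and the actor's forward vector, scaled by upward and forward distance scalable floats.
; Trace direction is that offset normalized.
; Trace length is the hypotenuse of both distances multiplied by a vertical check percent, keeping the trace tied to the travel tuning.
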(fn GetTargetRelativeLocation
  (params (ReferenceActor :type Actor%20Object%20Reference))
  (return (+ (* (ScalableFloat|GetValueatLevel (Variables|Default|GetHF_UpwardTravelDistance)) (Math|Vector|VectorUp)) (* (Transformation|GetActorForwardVector ReferenceActor) (ScalableFloat|GetValueatLevel (Variables|Default|GetHF_ForwardTravelDistance))))))

(fn GetTraceDirection
  (params (TargetingHandle :type Targeting%20Request%20Handle%20Structure%20%28by%20ref%29))
  (return (Math|Vector|Normalize (CallFunction|GetTargetRelativeLocation (Utilities|Struct|BreakTargetingSourceContext (TargetingSystem|TargetingTypes|GetTargetingSourceContext (Targeting|GetTargetingSubsystem TargetingHandle) TargetingHandle))))))

(fn GetTraceLength
  (params (TargetingHandle :type Targeting%20Request%20Handle%20Structure%20%28by%20ref%29))
  (return (* (Math|Float|Hypotenuse (ScalableFloat|GetValueatLevel (Variables|Default|GetHF_ForwardTravelDistance)) (ScalableFloat|GetValueatLevel (Variables|Default|GetHF_UpwardTravelDistance))) (ScalableFloat|GetValueatLevel (Variables|Default|GetHF_VerticalCheckPercent)))))
